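# Material function for the RL249 wheel's animated glowing decal, overriding the simple wheel parent's emissive, base color, opacity mask and world position offset.
# Full detail tints mask channels with rim colors, though a blue-minus-itself subtraction zeroes the blend toward the panning layer.
# A sine wave scaled by vertex blue drives local-Y vertex offset.
# A static LOD switch selects an alternative setup, with rotating and static LOD-texture masks and optional vertex offset.
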MaterialFunction /VehicleCosmetics/Wheels/RL249/Materials/MF_Wheel_RL249_Decal
graph {nodes [
  n1 = Constant()
  n2 = MaterialFunctionCall() {materialFunction=&MF_WheelParent_Simple}
  n3 = SetMaterialAttributes!(
   MaterialAttributes=@n2.MaterialAttributes
   EmissiveColor=NamedRerouteUsage()
   BaseColor=@n1
   WorldPositionOffset=StaticSwitchParameter(True=NamedRerouteUsage(), False=Constant()) {parameterName=LOD_EnableWPO}
   OpacityMask=NamedRerouteUsage()
  ) {attributeSetTypes [
    #1
    #2
    #3
    #4
   ]}
  n4 = SetMaterialAttributes(
   MaterialAttributes=@n2.MaterialAttributes
   EmissiveColor=NamedRerouteUsage()
   BaseColor=@n1
   WorldPositionOffset=NamedRerouteUsage()
   OpacityMask=NamedRerouteUsage()
  ) {attributeSetTypes [
    #1
    #2
    #3
    #4
   ]}
  FunctionOutput(None=StaticSwitchParameter(True=@n3, False=@n4) {parameterName=LOD})
  Multiply()
  n5 = TextureSample() {samplerType=SAMPLERTYPE_LinearColor, texture=/VehicleCosmetics/Wheels/RL249/Textures/T_Tire_RL249_Decals_M}
  n6 = Clamp(None=Subtract(A=@n5.B, B=@n5.B))
  NamedRerouteDeclaration(None=Add(A=@n5.R, B=@n6)) {name=OpacityMask, nodeColor=<rgba 0 0.4375 1 1>}
  n7 = Multiply(A=VectorParameter() {parameterName=Rim_ColorA, defaultValue=<rgba 0 0.3325 0.1746 1>}.RGB, B=@n5.R)
  n8 = VectorParameter() {parameterName=Rim_ColorA_PBR, defaultValue=<rgba 5 0 0 1>}
  n9 = Time()
  n10 = Multiply(
   A=VectorParameter() {parameterName=Rim_ColorB, defaultValue=<rgba 1 0.1329 0.0666 1>}.RGB
   B=TextureSample(UVs=Panner(Time=@n9) {speedX=-0.5}) {samplerType=SAMPLERTYPE_LinearColor, texture=&T_Tire_RL249_Decals_M}.G
  )
  n11 = VectorParameter() {parameterName=Rim_ColorB_PBR, defaultValue=<rgba 3 0 0 1>}
  NamedRerouteDeclaration(None=LinearInterpolate(A=Multiply(A=@n7, B=@n8.R), B=Multiply(A=@n10, B=@n11.R), Alpha=@n6)) {name=Emissive, nodeColor=<rgba 0.7578 1 0 1>}
  n12 = Time()
  n13 = Multiply(A=VertexColor().B, B=Add(A=ScalarParameter() {defaultValue=1, parameterName=DecalHeightOffset}, B=Sine(None=@n12)))
  n14 = Constant3Vector() {constant=<rgba 0 1 0 0>}
  NamedRerouteDeclaration(None=Transform(None=Multiply(A=@n13, B=@n14)) {transformSourceType=TRANSFORMSOURCE_Local}) {name=WPO, nodeColor=<rgba 0.9219 0 1 1>}
  n15 = Time()
  n16 = ScalarParameter() {defaultValue=1, parameterName=LOD_Rotator_speed}
  n17 = TextureSample(UVs=Rotator(Time=Multiply(A=@n15, B=@n16)) {centerX=0.745, centerY=0.305, speed=-2}) {
   samplerSource=SSM_Clamp_WorldGroupSettings
   samplerType=SAMPLERTYPE_LinearColor
   texture=/VehicleCosmetics/Wheels/RL249/Textures/T_Tire_RL249_LOD_M
  }
  n18 = Clamp(None=MaterialFunctionCall!(Base=@n17.G, Add=Constant() {r=20}) {materialFunction=/Engine/Functions/Engine_MaterialFunctions02/Math/MultiplyAdd}.Result)
  n19 = TextureSample() {
   samplerSource=SSM_Clamp_WorldGroupSettings
   samplerType=SAMPLERTYPE_LinearColor
   texture=&T_Tire_RL249_LOD_M
  }
  NamedRerouteDeclaration(None=Add(A=@n18, B=@n19.R)) {name=Opacity_LOD, nodeColor=<rgba 0.4531 1 0 1>}
  n20 = Multiply(
   A=Multiply(A=@n17.G, B=VectorParameter() {parameterName=Rim_ColorB, defaultValue=<rgba 1 0.4357 0.1239 1>}.RGB)
   B=VectorParameter() {parameterName=Rim_ColorB_PBR, defaultValue=<rgba 3 0 0 1>}.R
  )
  n21 = Multiply(A=@n19.R, B=VectorParameter() {parameterName=Rim_ColorA, defaultValue=<rgba 0 0.3325 0.1746 1>}.RGB)
  n22 = VectorParameter() {parameterName=Rim_ColorA_PBR, defaultValue=<rgba 5 0 0 1>}
  NamedRerouteDeclaration(None=Add(A=Multiply(A=@n18, B=@n20), B=Multiply(A=@n21, B=@n22.R))) {name=Emissive_LOD, nodeColor=<rgba 1 0 0.7734 1>}
 ]}
