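# Material function for the RL249 wheel rim that adds optional animated vertex motion to the attributes from MF_WheelParent_Simple.
# A time-driven sine, scaled by vertex color blue, oscillates vertices along the local Y axis, transformed out of local space.
# The EnableWPO static switch selects this offset or zero, and vertex color red fades it toward zero before it overrides World Position Offset.
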MaterialFunction /VehicleCosmetics/Wheels/RL249/Materials/MF_Wheel_RL249_rim
graph {nodes [
  n1 = VertexColor()
  n2 = Multiply(A=@n1.B, B=Sine(None=Time()))
  n3 = Constant3Vector() {constant=<rgba 0 1 0 0>}
  n4 = StaticSwitchParameter(True=Transform(None=Multiply(A=@n2, B=@n3)) {transformSourceType=TRANSFORMSOURCE_Local}, False=Constant()) {parameterName=EnableWPO}
  FunctionOutput(None=SetMaterialAttributes!(
    MaterialAttributes=MaterialFunctionCall() {materialFunction=&MF_WheelParent_Simple}.MaterialAttributes
    WorldPositionOffset=LinearInterpolate(A=@n4, Alpha=@n1.R) {constB=0}
   ) {attributeSetTypes [#1]})
 ]}
